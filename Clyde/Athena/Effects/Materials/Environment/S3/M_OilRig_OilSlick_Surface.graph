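# Unlit translucent material for an oil slick surface, evidently on oil rig water, usable on instanced and Niagara meshes.
# Panning slick and distortion textures warp a mask, forming broken, animated opacity.
# A view-angle factor hue-shifts the distorted slick texture for iridescence, scales it by time-of-day brightness and blends toward a two-color tint.
# Vertex offsets combine camera offset, depth-faded height push and a switchable water displacement term.
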Material /Clyde/Athena/Effects/Materials/Environment/S3/M_OilRig_OilSlick_Surface
graph {
 properties {
  bUsedWithInstancedStaticMeshes=true
  bUsedWithNiagaraMeshParticles=true
  blendMode=BLEND_Translucent
  numCustomizedUVs=2
  pixelDepthOffsetMode=PDOM_Legacy
  shadingModel=MSM_Unlit
  translucencyLightingMode=TLM_Surface
  usedShadingModels=Unlit
 }
 nodes [
  n1 = ScalarParameter() {defaultValue=0.4, group=HeightOffset, parameterName=Length, sliderMax=10000}
  n2 = ScalarParameter() {defaultValue=5000, group=HeightOffset, parameterName=Offset, sliderMax=10000}
  n3 = StaticBool() {value=true}
  n4 = ScalarParameter() {defaultValue=10, parameterName=CameraOffset}
  n5 = ScalarParameter() {defaultValue=1, parameterName=ZOffset}
  n6 = CameraPositionWS()
  n7 = ParticlePositionWS()
  n8 = ObjectPositionWS()
  n9 = FeatureLevelSwitch(Default=@n7, ES2=@n8, ES3_1=@n8, SM4_REMOVED=@n7)
  n10 = Clamp(None=ComponentMask(None=Subtract(A=@n6, B=@n9)) {b=true}) {minDefault=-1}
  n11 = Add(
   A=Multiply(B=VectorParameter() {parameterName=Water Displacement Multiplier, defaultValue=<rgba 1 1 1 0>}.RGB)
   B=MaterialFunctionCall(Z=Multiply(A=@n5, B=@n10)) {materialFunction=&MakeFloat3}.Result
  )
  n12 = Constant()
  Add(
   A=Multiply(
    A=MaterialFunctionCall(Fade Length=@n1, Fade Offset=@n2, For Vertex Shader=@n3) {materialFunction=&CameraDepthFade}.Result
    B=VectorParameter() {group=HeightOffset, parameterName=Z_Push, defaultValue=<rgba 0 0 300 0>}.RGB
   )
   B=Add(
    A=MaterialFunctionCall(Offset Amount=@n4) {materialFunction=/Engine/Functions/Engine_MaterialFunctions02/WorldPositionOffset/CameraOffset}.Clamped Camera Offset
    B=StaticSwitchParameter(True=@n11, False=@n12) {parameterName=UseWaterDisplacement}
   )
  )
  n13 = TextureCoordinate() {coordinateIndex=1}
  n14 = TextureSample(UVs=@n13) {samplerType=SAMPLERTYPE_Normal, texture=/Clyde/Packages/Fortnite_UniqueMaterials_1/Textures/T_Water2_N}
  n15 = ComponentMask(None=@n14.RGB) {g=true, r=true}
  n16 = ScalarParameter() {defaultValue=0.05, group=Deep Water, parameterName=Noise 1}
  n17 = TextureSample(UVs=Add(A=Multiply(A=@n15, B=@n16), B=@n13)) {samplerType=SAMPLERTYPE_Normal, texture=&T_Water2_N}
  n18 = TextureCoordinate() {coordinateIndex=1}
  n19 = VertexInterpolator(VS=@n18)
  n20 = TextureSample(UVs=VertexInterpolator(VS=@n18).PS) {samplerType=SAMPLERTYPE_Normal, texture=&T_Water2_N}
  n21 = TextureSample(UVs=VertexInterpolator(VS=@n18).PS) {samplerType=SAMPLERTYPE_Normal, texture=&T_Water2_N}
  n22 = MaterialFunctionCall(
   BaseNormal=MaterialFunctionCall(
    Normal=TextureSample(UVs=@n19.PS) {samplerType=SAMPLERTYPE_Normal, texture=/Clyde/Packages/Fortnite_UniqueMaterials_1/Textures/T_WaterSoft_N}.RGB
    Flatness=ScalarParameter() {defaultValue=0.9115, parameterName=NormalFlaten_M3}
   ) {materialFunction=&FlattenNormal}.Result
   AdditionalNormal=MaterialFunctionCall(
    Normal=MaterialFunctionCall(BaseNormal=@n20.RGB, AdditionalNormal=@n21.RGB) {materialFunction=&BlendAngleCorrectedNormals}.Result
    Flatness=ScalarParameter() {parameterName=NormalFlatten}
   ) {materialFunction=&FlattenNormal}.Result
  ) {materialFunction=&BlendAngleCorrectedNormals}
  n23 = ShadingPathSwitch(
   Default=MaterialFunctionCall(
    Normal=MaterialFunctionCall(BaseNormal=@n14.RGB, AdditionalNormal=@n17.RGB) {materialFunction=&BlendAngleCorrectedNormals}.Result
    Flatness=ScalarParameter() {parameterName=NormalFlatten}
   ) {materialFunction=&FlattenNormal}.Result
   Forward=@n22.Result
   Mobile=@n22.Result
  )
  n24 = Saturate(None=ComponentMask(None=@n23) {b=true})
  LinearInterpolate(
   A=ScalarParameter() {defaultValue=0.01, parameterName=Roughness Max}
   B=ScalarParameter() {defaultValue=0.25, parameterName=Roughness Min}
   Alpha=@n24
  ) {constA=0.1, constB=0.01}
  LinearInterpolate(B=ScalarParameter() {defaultValue=0.5, parameterName=Metallic Max}, Alpha=@n24) {constB=0.5}
  MaterialFunctionCall()
  MaterialFunctionCall()
  n25 = ScalarParameter() {defaultValue=0.6, parameterName=Alpha}
  n26 = TextureCoordinate() {uTiling=2, vTiling=2}
  n27 = Multiply(
   A=@n26
   B=AppendVector(A=ScalarParameter() {defaultValue=1, parameterName=Tiling_U}, B=ScalarParameter() {defaultValue=1, parameterName=Tiling_V})
  )
  n28 = ComponentMask(None=VectorParameter() {parameterName=DistrotSpeed, defaultValue=<rgba 0.1 0.1 0 1>}.RGB) {g=true, r=true}
  n29 = Multiply(A=Time(), B=ScalarParameter() {defaultValue=1, parameterName=PanSpeed})
  n30 = TextureSample(UVs=Panner(Coordinate=@n27, Time=@n29, Speed=@n28) {speedX=-0.05, speedY=-0.1}) {samplerSource=SSM_Wrap_WorldGroupSettings, texture=/Clyde/Athena/Effects/Textures/Environments/S3/T_GasStation_OilSlick}
  n31 = ScalarParameter() {defaultValue=1, parameterName=TilingMult2}
  n32 = TextureSample(UVs=Panner(Coordinate=Multiply(A=@n26, B=@n31), Time=@n29, Speed=Divide(A=@n28)) {speedX=-0.05, speedY=-0.1}) {samplerSource=SSM_Wrap_WorldGroupSettings, texture=&T_GasStation_OilSlick}
  n33 = TextureSample(UVs=Panner(Coordinate=@n27, Time=@n29, Speed=@n28) {speedX=-0.05, speedY=-0.1}) {texture=/Clyde/Effects/Fort_Effects/Textures/Tile/T_BlasterBeamDistort}
  n34 = Multiply(A=@n33.RGB, B=ScalarParameter() {defaultValue=1, parameterName=Distport power})
  n35 = Add(
   A=Multiply(A=Multiply(A=@n30.G, B=@n32.G), B=ScalarParameter() {defaultValue=1, parameterName=Distport power})
   B=Add(A=TextureCoordinate(), B=ComponentMask(None=@n34) {g=true, r=true})
  )
  n36 = VectorParameter() {parameterName=OffsetDistort}
  n37 = TextureSampleParameter2D(UVs=Add(A=@n35, B=ComponentMask(None=@n36.RGB) {g=true, r=true})) {parameterName=OilSlickMask, texture=/Clyde/Athena/Effects/Textures/Environments/S3/T_OilRig_OilSlick}
  n38 = Power(
   Base=Add(A=ComponentMask(None=@n33.RGB) {b=true}, B=ScalarParameter() {defaultValue=0.25, parameterName=Param_5})
   Exp=ScalarParameter() {defaultValue=10, parameterName=Param_1}
  )
  Multiply(
   A=Multiply(A=Multiply(A=@n25, B=@n30.G), B=StaticComponentMaskParameter(None=@n37.RGB) {defaultG=true, parameterName=Alpha Mask Channel})
   B=Clamp(None=LinearInterpolate(Alpha=@n38) {constA=1, constB=0})
  )
  Reroute(None=@n23)
  n39 = MaterialFunctionCall(
   Color_Morning=ScalarParameter() {defaultValue=1, parameterName=Brightness_Morning}
   Color_Day=ScalarParameter() {defaultValue=1.4, parameterName=Brightness_Day}
   Color_Evening=ScalarParameter() {defaultValue=1, parameterName=Brightness_Evening}
   Color_Night=ScalarParameter() {defaultValue=0.65, parameterName=Brightness_Night}
  ) {materialFunction=&Color_Time_of_Day}
  n40 = CameraVectorWS()
  n41 = Constant3Vector() {constant=<rgba 0 0 1 0>}
  n42 = Clamp(None=Power(Base=DotProduct(A=@n40, B=@n41), Exp=ScalarParameter() {defaultValue=5, parameterName=Exp}) {constExponent=0.5})
  n43 = Multiply(
   A=Add(
    A=MaterialFunctionCall() {materialFunction=&Swizzle}.YX
    B=ScalarParameter() {defaultValue=0.2, parameterName=ColorScroll}
   )
   B=ComponentMask(None=VectorParameter() {parameterName=RainbowTile}.RGB) {g=true, r=true}
  )
  n44 = TextureSample(UVs=Multiply(A=Add(A=@n43, B=@n35), B=ScalarParameter() {defaultValue=5, parameterName=DistortExtra})) {texture=&T_GasStation_OilSlick}
  n45 = MaterialFunctionCall(Hue Shift Percentage=@n42, Texture=@n44.RGB) {materialFunction=&HueShift}
  n46 = VectorParameter() {parameterName=ColorA, defaultValue=<rgba 0.0008 0.1042 0.0874 1>}
  n47 = VectorParameter() {parameterName=ColorB, defaultValue=<rgba 0.0076 0.974 0.8173 1>}
  Reroute(None=LinearInterpolate(A=Multiply(A=@n39.Result, B=@n45.Result), B=LinearInterpolate(A=@n46.RGB, B=@n47.RGB, Alpha=@n44.B), Alpha=@n42))
  n48 = WorldPosition()
  Rotator(
   Coordinate=Divide(A=ComponentMask(None=@n48.XYZ) {g=true, r=true}, B=ScalarParameter() {defaultValue=5000, parameterName=XTile})
   Time=Multiply(A=Time(), B=ScalarParameter() {defaultValue=-0.001, parameterName=Primary Rotation Speed})
  )
  ScalarParameter() {group=Water, parameterName=WaterBodyIndex}
  ScalarParameter() {group=Water, parameterName=WaterDepth}
  ScalarParameter() {parameterName=Specular}
  ScalarParameter() {defaultValue=1, parameterName=TilingMult}
  StaticBoolParameter() {defaultValue=true, parameterName=Ignore XY Gerstner Offsets}
 ]
}
